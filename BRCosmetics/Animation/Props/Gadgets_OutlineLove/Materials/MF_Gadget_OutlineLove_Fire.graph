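# Material function that renders procedural animated fire for the OutlineLove gadget cosmetic, outputting color plus static and scrolling turbulent UVs.
# Centered UVs are spherized, rotated and stretched, then distorted by custom HLSL that layers rotated sine waves over scrolling octaves.
# Distance-based core falloff with vertical attenuation, a texture breakup pattern and a soft glow shape the tinted luminance.
# An optional exponential range compression and a contrast power finish the result.
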MaterialFunction /BRCosmetics/Animation/Props/Gadgets_OutlineLove/Materials/MF_Gadget_OutlineLove_Fire
graph {expressions {
  MaterialExpressionAdd_0 {
   class=MaterialExpressionAdd
   inputs [
    A {from=MaterialExpressionMultiply_8}
    B {from=""}
   ]
  }
  MaterialExpressionAdd_1 {
   class=MaterialExpressionAdd
   inputs [
    A {from=MaterialExpressionNamedRerouteUsage_15}
    B {from=MaterialExpressionCustom_0}
   ]
   properties {desc="Reverse scroll to output turbulent UV\r\nin the same range as the input UV"}
  }
  MaterialExpressionAdd_3 {
   class=MaterialExpressionAdd
   inputs [
    A {from=MaterialExpressionNamedRerouteUsage_28}
    B {from=MaterialExpressionConvert_2}
   ]
  }
  MaterialExpressionAdd_4 {
   class=MaterialExpressionAdd
   inputs [
    A {from=MaterialExpressionDivide_15}
    B {from=""}
   ]
  }
  MaterialExpressionAdd_5 {
   class=MaterialExpressionAdd
   inputs [
    A {from=MaterialExpressionMultiply_24}
    B {from=MaterialExpressionMultiply_25}
   ]
  }
  MaterialExpressionAdd_7 {
   class=MaterialExpressionAdd
   inputs [
    A {from=MaterialExpressionDotProduct_0}
    B {from=""}
   ]
  }
  MaterialExpressionAdd_8 {
   class=MaterialExpressionAdd
   inputs [
    A {from=MaterialExpressionMultiply_32}
    B {from=MaterialExpressionNamedRerouteUsage_57}
   ]
  }
  MaterialExpressionAppendVector_1 {
   class=MaterialExpressionAppendVector
   inputs [
    A {from=MaterialExpressionLinearInterpolate_3, output=""}
    B {from=MaterialExpressionLinearInterpolate_4, output=""}
   ]
  }
  MaterialExpressionComponentMask_10 {class=MaterialExpressionComponentMask, inputs [None {from=MaterialExpressionNamedRerouteUsage_40}], properties {g=true}}
  MaterialExpressionComponentMask_3 {class=MaterialExpressionComponentMask, inputs [None {from=MaterialExpressionNamedRerouteUsage_12}], properties {g=true}}
  MaterialExpressionComponentMask_4 {class=MaterialExpressionComponentMask, inputs [None {from=MaterialExpressionNamedRerouteUsage_11}], properties {r=true}}
  MaterialExpressionConstantBiasScale_0 {
   class=MaterialExpressionConstantBiasScale
   inputs [None {from=MaterialExpressionConvert_5}]
   properties {bias=-0.5, desc="UV in [-1, 1] range", scale=2}
  }
  MaterialExpressionConvert_2 {
   class=MaterialExpressionConvert
   inputs [None {from=MaterialExpressionMultiply_11}]
   properties {
    convertInputs [{
      type=Vector2
      expressionInput {
       expression=&self:MaterialExpressionMultiply_11
       inputName=None
       mask=0
       maskA=0
       maskB=0
       maskG=0
       maskR=0
       outputIndex=0
      }
     }]
    convertOutputs [{type=Vector2}]
   }
  }
  MaterialExpressionConvert_3 {
   class=MaterialExpressionConvert
   inputs [None {from=MaterialExpressionNamedRerouteUsage_50}]
   properties {
    convertInputs [{
      type=Vector2
      expressionInput {
       expression=&self:MaterialExpressionNamedRerouteUsage_50
       inputName=None
       mask=0
       maskA=0
       maskB=0
       maskG=0
       maskR=0
       outputIndex=0
      }
     }]
    convertOutputs [{type=Vector2, defaultValue=<rgba 1 0 0 1>}]
   }
  }
  MaterialExpressionConvert_4 {
   class=MaterialExpressionConvert
   inputs [None {from=MaterialExpressionNamedRerouteUsage_23}]
   properties {
    nodeName=Break Float2
    convertInputs [{
      type=Vector2
      expressionInput {
       expression=&self:MaterialExpressionNamedRerouteUsage_23
       inputName=None
       mask=0
       maskA=0
       maskB=0
       maskG=0
       maskR=0
       outputIndex=0
      }
     }]
    convertOutputs [{type=Scalar}, {type=Scalar}]
   }
  }
  MaterialExpressionConvert_5 {
   class=MaterialExpressionConvert
   inputs [
    None {from=MaterialExpressionConvert_4}
    None {from=MaterialExpressionOneMinus_3}
   ]
   properties {
    nodeName=Make Vector2
    convertInputs [
     {
      type=Scalar
      expressionInput {
       expression=&self:MaterialExpressionConvert_4
       inputName=None
       mask=0
       maskA=0
       maskB=0
       maskG=0
       maskR=0
       outputIndex=0
      }
     }
     {
      type=Scalar
      expressionInput {
       expression=&self:MaterialExpressionOneMinus_3
       inputName=None
       mask=0
       maskA=0
       maskB=0
       maskG=0
       maskR=0
       outputIndex=0
      }
     }
    ]
    convertOutputs [{type=Vector2}]
   }
  }
  MaterialExpressionConvert_6 {
   class=MaterialExpressionConvert
   inputs [None {from=MaterialExpressionNamedRerouteUsage_16}]
   properties {
    convertInputs [{
      type=Scalar
      expressionInput {
       expression=&self:MaterialExpressionNamedRerouteUsage_16
       inputName=None
       mask=0
       maskA=0
       maskB=0
       maskG=0
       maskR=0
       outputIndex=0
      }
     }]
    convertOutputs [{type=Vector2}]
   }
  }
  MaterialExpressionCustom_0 {
   class=MaterialExpressionCustom
   inputs [
    UV {from=MaterialExpressionSubtract_4, output=""}
    NumOctaves {from=MaterialExpressionNamedRerouteUsage_0, output=""}
    Amplitude {from=MaterialExpressionNamedRerouteUsage_1, output=""}
    Speed {from=MaterialExpressionNamedRerouteUsage_2, output=""}
    Frequency {from=MaterialExpressionNamedRerouteUsage_3, output=""}
    FrequencyMult {from=MaterialExpressionNamedRerouteUsage_4, output=""}
   ]
   properties {
    code="//Turbulence starting scale\r\nfloat freq = Frequency;\r\n\r\n//Turbulence rotation matrix\r\nfloat2x2 m_base_rot = {0.6, -0.8, 0.8, 0.6};\r\nfloat2x2 m_rot = m_base_rot;\r\n\r\nfloat2 p = UV;\r\n\r\n//Loop through turbulence octaves\r\nfor (float i=0.0; i<NumOctaves; i++)\r\n{\r\n    //Scroll along the rotated y coordinate\r\n    float phase = freq * mul(p, m_rot).y + Speed + i;\r\n    //Add a perpendicular sine wave offset\r\n    p += Amplitude * m_rot[0] * sin(phase) / freq;\r\n    \r\n    //Rotate for the next octave\r\n    m_rot = mul(m_rot, m_base_rot);\r\n    //Scale down for the next octave\r\n    freq *= FrequencyMult;\r\n}\r\n\r\nreturn p;"
    description=UV Turbulence
    outputType=CMOT_Float2
    inputs [
     {
      inputName=UV
      input {
       expression=&self:MaterialExpressionSubtract_4
       inputName=None
       mask=0
       maskA=0
       maskB=0
       maskG=0
       maskR=0
       outputIndex=0
      }
     }
     {
      inputName=NumOctaves
      input {
       expression=&self:MaterialExpressionNamedRerouteUsage_0
       inputName=None
       mask=0
       maskA=0
       maskB=0
       maskG=0
       maskR=0
       outputIndex=0
      }
     }
     {
      inputName=Amplitude
      input {
       expression=&self:MaterialExpressionNamedRerouteUsage_1
       inputName=None
       mask=0
       maskA=0
       maskB=0
       maskG=0
       maskR=0
       outputIndex=0
      }
     }
     {
      inputName=Speed
      input {
       expression=&self:MaterialExpressionNamedRerouteUsage_2
       inputName=None
       mask=0
       maskA=0
       maskB=0
       maskG=0
       maskR=0
       outputIndex=0
      }
     }
     {
      inputName=Frequency
      input {
       expression=&self:MaterialExpressionNamedRerouteUsage_3
       inputName=None
       mask=0
       maskA=0
       maskB=0
       maskG=0
       maskR=0
       outputIndex=0
      }
     }
     {
      inputName=FrequencyMult
      input {
       expression=&self:MaterialExpressionNamedRerouteUsage_4
       inputName=None
       mask=0
       maskA=0
       maskB=0
       maskG=0
       maskR=0
       outputIndex=0
      }
     }
    ]
   }
  }
  MaterialExpressionDivide_10 {
   class=MaterialExpressionDivide
   inputs [
    A {from=MaterialExpressionNamedRerouteUsage_29}
    B {from=MaterialExpressionConvert_3}
   ]
   properties {constB=1}
  }
  MaterialExpressionDivide_11 {
   class=MaterialExpressionDivide
   inputs [
    A {from=""}
    B {from=MaterialExpressionAdd_4}
   ]
   properties {constA=0.01}
  }
  MaterialExpressionDivide_12 {
   class=MaterialExpressionDivide
   inputs [
    A {from=""}
    B {from=MaterialExpressionMax_2}
   ]
  }
  MaterialExpressionDivide_13 {
   class=MaterialExpressionDivide
   inputs [
    A {from=MaterialExpressionNamedRerouteUsage_44}
    B {from=MaterialExpressionAdd_7}
   ]
  }
  MaterialExpressionDivide_15 {
   class=MaterialExpressionDivide
   inputs [
    A {from=MaterialExpressionNamedRerouteUsage_38}
    B {from=MaterialExpressionNamedRerouteUsage_55}
   ]
  }
  MaterialExpressionDivide_16 {
   class=MaterialExpressionDivide
   inputs [
    A {from=MaterialExpressionNamedRerouteUsage_52}
    B {from=MaterialExpressionNamedRerouteUsage_18}
   ]
   properties {constB=100}
  }
  MaterialExpressionDivide_2 {
   class=MaterialExpressionDivide
   inputs [
    A {from=""}
    B {from=MaterialExpressionAdd_0}
   ]
   properties {constA=0.5}
  }
  MaterialExpressionDivide_3 {
   class=MaterialExpressionDivide
   inputs [
    A {from=MaterialExpressionMultiply_9}
    B {from=MaterialExpressionNamedRerouteUsage_19}
   ]
  }
  MaterialExpressionDivide_7 {
   class=MaterialExpressionDivide
   inputs [
    A {from=MaterialExpressionConvert_6}
    B {from=MaterialExpressionNamedRerouteUsage_22}
   ]
  }
  MaterialExpressionDivide_9 {
   class=MaterialExpressionDivide
   inputs [
    A {from=MaterialExpressionLength_2}
    B {from=""}
   ]
   properties {constB=1.4}
  }
  MaterialExpressionDotProduct_0 {
   class=MaterialExpressionDotProduct
   inputs [
    A {from=MaterialExpressionNamedRerouteUsage_32, output=""}
    B {from=MaterialExpressionNamedRerouteUsage_32, output=""}
   ]
   properties {}
  }
  MaterialExpressionExponential_1 {class=MaterialExpressionExponential, inputs [None {from=MaterialExpressionMultiply_21, output=""}], properties {}}
  MaterialExpressionFunctionInput_0 {
   class=MaterialExpressionFunctionInput
   inputs [Preview {from=MaterialExpressionTextureCoordinate_0, output=""}]
   properties {bUsePreviewValueAsDefault=true, inputName=UV, inputType=FunctionInput_Vector2}
  }
  MaterialExpressionFunctionOutput_0 {class=MaterialExpressionFunctionOutput, inputs [None {from=MaterialExpressionPower_6, output=""}], properties {outputName="Result (V3)"}}
  MaterialExpressionFunctionOutput_1 {
   class=MaterialExpressionFunctionOutput
   inputs [None {from=MaterialExpressionNamedRerouteUsage_64, output=""}]
   properties {outputName="Turbulence UV (Static) (V2)", sortPriority=10}
  }
  MaterialExpressionFunctionOutput_2 {
   class=MaterialExpressionFunctionOutput
   inputs [None {from=MaterialExpressionNamedRerouteUsage_63, output=""}]
   properties {outputName="Turbulence UV (Scrolling) (V2)", sortPriority=20}
  }
  MaterialExpressionLength_1 {class=MaterialExpressionLength, inputs [None {from=MaterialExpressionMin_0, output=""}], properties {}}
  MaterialExpressionLength_2 {class=MaterialExpressionLength, inputs [None {from=MaterialExpressionNamedRerouteUsage_46, output=""}], properties {}}
  MaterialExpressionLength_3 {class=MaterialExpressionLength, inputs [None {from=MaterialExpressionAdd_3, output=""}], properties {}}
  MaterialExpressionLinearInterpolate_0 {
   class=MaterialExpressionLinearInterpolate
   inputs [
    A {from=""}
    B {from=MaterialExpressionTextureSample_1, output=RGB}
    Alpha {from=MaterialExpressionNamedRerouteUsage_31}
   ]
   properties {constA=1}
  }
  MaterialExpressionLinearInterpolate_3 {
   class=MaterialExpressionLinearInterpolate
   inputs [
    A {from=""}
    B {from=MaterialExpressionSubtract_3}
    Alpha {from=MaterialExpressionNamedRerouteUsage_54}
   ]
   properties {constA=1, desc=U Scale}
  }
  MaterialExpressionLinearInterpolate_4 {
   class=MaterialExpressionLinearInterpolate
   inputs [
    A {from=""}
    B {from=MaterialExpressionOneMinus_4}
    Alpha {from=MaterialExpressionNamedRerouteUsage_59}
   ]
   properties {constA=1, desc=V Scale}
  }
  MaterialExpressionLinearInterpolate_8 {
   class=MaterialExpressionLinearInterpolate
   inputs [
    A {from=MaterialExpressionNamedRerouteUsage_53}
    B {from=MaterialExpressionOneMinus_9}
    Alpha {from=MaterialExpressionNamedRerouteUsage_24}
   ]
  }
  MaterialExpressionMaterialFunctionCall_10 {
   class=MaterialExpressionMaterialFunctionCall
   inputs [
    Angle {from=MaterialExpressionNamedRerouteUsage_10}
    UV/2DVector {from=MaterialExpressionMultiply_16}
   ]
   properties {materialFunction=/Game/Characters/Player/Male/Medium/Bodies/M_MED_Faux/Materials/MF_2DRotate}
  }
  MaterialExpressionMax_0 {
   class=MaterialExpressionMax
   inputs [
    A {from=MaterialExpressionAdd_8}
    B {from=""}
   ]
   properties {constB=0}
  }
  MaterialExpressionMax_1 {
   class=MaterialExpressionMax
   inputs [
    A {from=MaterialExpressionSubtract_10}
    B {from=""}
   ]
   properties {constB=0}
  }
  MaterialExpressionMax_2 {
   class=MaterialExpressionMax
   inputs [
    A {from=MaterialExpressionPower_5}
    B {from=""}
   ]
   properties {constB=0.0001}
  }
  MaterialExpressionMin_0 {
   class=MaterialExpressionMin
   inputs [
    A {from=MaterialExpressionNamedRerouteUsage_29, output=""}
    B {from=MaterialExpressionDivide_10, output=""}
   ]
   properties {}
  }
  MaterialExpressionMultiply_0 {
   class=MaterialExpressionMultiply
   inputs [
    A {from=MaterialExpressionDivide_12}
    B {from=MaterialExpressionNamedRerouteUsage_35}
   ]
  }
  MaterialExpressionMultiply_1 {
   class=MaterialExpressionMultiply
   inputs [
    A {from=MaterialExpressionScalarParameter_3}
    B {from=MaterialExpressionNamedRerouteUsage_60}
   ]
  }
  MaterialExpressionMultiply_11 {
   class=MaterialExpressionMultiply
   inputs [
    A {from=MaterialExpressionNamedRerouteUsage_48}
    B {from=MaterialExpressionMultiply_7}
   ]
  }
  MaterialExpressionMultiply_13 {
   class=MaterialExpressionMultiply
   inputs [
    A {from=MaterialExpressionLength_1}
    B {from=MaterialExpressionNamedRerouteUsage_20}
   ]
  }
  MaterialExpressionMultiply_16 {
   class=MaterialExpressionMultiply
   inputs [
    A {from=MaterialExpressionMultiply_17}
    B {from=MaterialExpressionNamedRerouteUsage_36}
   ]
  }
  MaterialExpressionMultiply_17 {
   class=MaterialExpressionMultiply
   inputs [
    A {from=MaterialExpressionPower_4}
    B {from=MaterialExpressionNamedRerouteUsage_9}
   ]
   properties {constB=5}
  }
  MaterialExpressionMultiply_2 {
   class=MaterialExpressionMultiply
   inputs [
    A {from=MaterialExpressionScalarParameter_8}
    B {from=MaterialExpressionTime_0}
   ]
  }
  MaterialExpressionMultiply_21 {
   class=MaterialExpressionMultiply
   inputs [
    A {from=MaterialExpressionNamedRerouteUsage_53}
    B {from=""}
   ]
   properties {constB=-1}
  }
  MaterialExpressionMultiply_22 {
   class=MaterialExpressionMultiply
   inputs [
    A {from=MaterialExpressionLength_3}
    B {from=MaterialExpressionNamedRerouteUsage_33}
   ]
  }
  MaterialExpressionMultiply_24 {
   class=MaterialExpressionMultiply
   inputs [
    A {from=MaterialExpressionNamedRerouteUsage_39}
    B {from=MaterialExpressionNamedRerouteUsage_39}
   ]
  }
  MaterialExpressionMultiply_25 {
   class=MaterialExpressionMultiply
   inputs [
    A {from=MaterialExpressionMax_1}
    B {from=MaterialExpressionNamedRerouteUsage_27}
   ]
  }
  MaterialExpressionMultiply_26 {
   class=MaterialExpressionMultiply
   inputs [
    A {from=MaterialExpressionNamedRerouteUsage_41}
    B {from=MaterialExpressionNamedRerouteUsage_42}
   ]
  }
  MaterialExpressionMultiply_29 {
   class=MaterialExpressionMultiply
   inputs [
    A {from=MaterialExpressionDivide_13}
    B {from=MaterialExpressionNamedRerouteUsage_17}
   ]
  }
  MaterialExpressionMultiply_3 {
   class=MaterialExpressionMultiply
   inputs [
    A {from=MaterialExpressionScalarParameter_7}
    B {from=MaterialExpressionNamedRerouteUsage_61}
   ]
  }
  MaterialExpressionMultiply_31 {
   class=MaterialExpressionMultiply
   inputs [
    A {from=MaterialExpressionCustom_0}
    B {from=MaterialExpressionNamedRerouteUsage_62}
   ]
   properties {desc="Scrolling trubulent UV, scale\r\ndepends on turbulence frequency"}
  }
  MaterialExpressionMultiply_32 {
   class=MaterialExpressionMultiply
   inputs [
    A {from=MaterialExpressionMultiply_26}
    B {from=MaterialExpressionNamedRerouteUsage_45}
   ]
  }
  MaterialExpressionMultiply_6 {
   class=MaterialExpressionMultiply
   inputs [
    A {from=MaterialExpressionSmoothStep_0}
    B {from=""}
   ]
   properties {constA=1.5, constB=1.5}
  }
  MaterialExpressionMultiply_7 {
   class=MaterialExpressionMultiply
   inputs [
    A {from=MaterialExpressionNamedRerouteUsage_49}
    B {from=""}
   ]
   properties {constB=2}
  }
  MaterialExpressionMultiply_8 {
   class=MaterialExpressionMultiply
   inputs [
    A {from=MaterialExpressionComponentMask_4}
    B {from=MaterialExpressionComponentMask_4}
   ]
  }
  MaterialExpressionMultiply_9 {
   class=MaterialExpressionMultiply
   inputs [
    A {from=MaterialExpressionNamedRerouteUsage_47}
    B {from=MaterialExpressionNamedRerouteUsage_13}
   ]
  }
  MaterialExpressionNamedRerouteDeclaration_0 {
   class=MaterialExpressionNamedRerouteDeclaration
   inputs [None {from=MaterialExpressionVectorParameter_1, output=Num Octaves}]
   properties {name="Num Octaves (S)", nodeColor=<rgba 1 0 0.8203 1>}
  }
  MaterialExpressionNamedRerouteDeclaration_1 {
   class=MaterialExpressionNamedRerouteDeclaration
   inputs [None {from=MaterialExpressionVectorParameter_1, output=Wave Amplitude}]
   properties {name="Amplitude (S)", nodeColor=<rgba 0 1 0.8594 1>}
  }
  MaterialExpressionNamedRerouteDeclaration_10 {
   class=MaterialExpressionNamedRerouteDeclaration
   inputs [None {from=MaterialExpressionMaterialFunctionCall_10, output=Result}]
   properties {name="Transformed UV (V2)", nodeColor=<rgba 0 0.6484 1 1>}
  }
  MaterialExpressionNamedRerouteDeclaration_11 {
   class=MaterialExpressionNamedRerouteDeclaration
   inputs [None {from=MaterialExpressionDivide_7}]
   properties {name="Scroll (V2)", nodeColor=<rgba 0.9688 1 0 1>}
  }
  MaterialExpressionNamedRerouteDeclaration_12 {
   class=MaterialExpressionNamedRerouteDeclaration
   inputs [None {from=MaterialExpressionScalarParameter_15}]
   properties {name="Warped UV Scale (S)", nodeColor=<rgba 0 0.6484 1 1>}
  }
  MaterialExpressionNamedRerouteDeclaration_13 {
   class=MaterialExpressionNamedRerouteDeclaration
   inputs [None {from=MaterialExpressionFunctionInput_0}]
   properties {name="UV In (V2)", nodeColor=<rgba 1 0 0.6094 1>}
  }
  MaterialExpressionNamedRerouteDeclaration_14 {
   class=MaterialExpressionNamedRerouteDeclaration
   inputs [None {from=MaterialExpressionAdd_1}]
   properties {name="Turbulent UV (Static) (V2)", desc="Use for UV-based (or distance) calculations", nodeColor=<rgba 0.4766 0 1 1>}
  }
  MaterialExpressionNamedRerouteDeclaration_15 {
   class=MaterialExpressionNamedRerouteDeclaration
   inputs [None {from=MaterialExpressionScalarParameter_26}]
   properties {name="Core Radius (S)", nodeColor=<rgba 0 1 0.0859 1>}
  }
  MaterialExpressionNamedRerouteDeclaration_16 {
   class=MaterialExpressionNamedRerouteDeclaration
   inputs [None {from=MaterialExpressionSubtract_9}]
   properties {name="Distance to Core (S)", nodeColor=<rgba 1 0 0.2344 1>}
  }
  MaterialExpressionNamedRerouteDeclaration_17 {
   class=MaterialExpressionNamedRerouteDeclaration
   inputs [None {from=MaterialExpressionMultiply_22}]
   properties {name="Distance to Source (S)", nodeColor=<rgba 0 0.5547 1 1>}
  }
  MaterialExpressionNamedRerouteDeclaration_18 {
   class=MaterialExpressionNamedRerouteDeclaration
   inputs [None {from=MaterialExpressionDivide_11}]
   properties {name="Falloff RGB (V3)", nodeColor=<rgba 0.4062 1 0 1>}
  }
  MaterialExpressionNamedRerouteDeclaration_19 {
   class=MaterialExpressionNamedRerouteDeclaration
   inputs [None {from=MaterialExpressionMultiply_0}]
   properties {name="Core Luminance (S)", nodeColor=<rgba 1 0 0.1875 1>}
  }
  MaterialExpressionNamedRerouteDeclaration_2 {
   class=MaterialExpressionNamedRerouteDeclaration
   inputs [None {from=MaterialExpressionMultiply_1}]
   properties {name="Turbulence Speed (S)", nodeColor=<rgba 1 0.5391 0 1>}
  }
  MaterialExpressionNamedRerouteDeclaration_20 {
   class=MaterialExpressionNamedRerouteDeclaration
   inputs [None {from=MaterialExpressionLinearInterpolate_0}]
   properties {name="Breakup Pattern (V3)", nodeColor=<rgba 1 0.9844 0 1>}
  }
  MaterialExpressionNamedRerouteDeclaration_21 {
   class=MaterialExpressionNamedRerouteDeclaration
   inputs [None {from=MaterialExpressionMultiply_3}]
   properties {name="Scroll Speed (S)", nodeColor=<rgba 0 0.4375 1 1>}
  }
  MaterialExpressionNamedRerouteDeclaration_22 {
   class=MaterialExpressionNamedRerouteDeclaration
   inputs [None {from=MaterialExpressionScalarParameter_38}]
   properties {name="Soft Glow Intensity (S)", nodeColor=<rgba 0.7578 1 0 1>}
  }
  MaterialExpressionNamedRerouteDeclaration_23 {
   class=MaterialExpressionNamedRerouteDeclaration
   inputs [None {from=MaterialExpressionScalarParameter_42}]
   properties {name="Breakup Pattern Scale (S)", nodeColor=<rgba 0.9219 0 1 1>}
  }
  MaterialExpressionNamedRerouteDeclaration_24 {
   class=MaterialExpressionNamedRerouteDeclaration
   inputs [None {from=MaterialExpressionTextureObjectParameter_0}]
   properties {name="Breakup Pattern Map (TO)", nodeColor=<rgba 0 1 0.6016 1>}
  }
  MaterialExpressionNamedRerouteDeclaration_25 {
   class=MaterialExpressionNamedRerouteDeclaration
   inputs [None {from=MaterialExpressionConstantBiasScale_0}]
   properties {name="Centered UV (V2)", nodeColor=<rgba 0.6875 1 0 1>}
  }
  MaterialExpressionNamedRerouteDeclaration_26 {
   class=MaterialExpressionNamedRerouteDeclaration
   inputs [None {from=MaterialExpressionAppendVector_1}]
   properties {name="UV Scale Factor (V2)", nodeColor=<rgba 0.9922 0 1 1>}
  }
  MaterialExpressionNamedRerouteDeclaration_27 {
   class=MaterialExpressionNamedRerouteDeclaration
   inputs [None {from=MaterialExpressionMultiply_31}]
   properties {name="Turbulent UV (Scrolling) (V2)", desc=Use as texture coordinate for 2D patterns, nodeColor=<rgba 0 0.7656 1 1>}
  }
  MaterialExpressionNamedRerouteDeclaration_28 {
   class=MaterialExpressionNamedRerouteDeclaration
   inputs [None {from=MaterialExpressionMax_0}]
   properties {name="Fire Luminance (V3)", nodeColor=<rgba 0 1 0.8828 1>}
  }
  MaterialExpressionNamedRerouteDeclaration_3 {
   class=MaterialExpressionNamedRerouteDeclaration
   inputs [None {from=MaterialExpressionVectorParameter_1, output=Base Frequency}]
   properties {name="Frequency (S)", nodeColor=<rgba 0.2188 0 1 1>}
  }
  MaterialExpressionNamedRerouteDeclaration_30 {
   class=MaterialExpressionNamedRerouteDeclaration
   inputs [None {from=MaterialExpressionVectorParameter_0, output=RGB}]
   properties {name="Fire Color (V3)", nodeColor=<rgba 0.2422 0 1 1>}
  }
  MaterialExpressionNamedRerouteDeclaration_31 {
   class=MaterialExpressionNamedRerouteDeclaration
   inputs [None {from=MaterialExpressionScalarParameter_0}]
   properties {name="Core Brightness (S)", nodeColor=<rgba 0.0781 1 0 1>}
  }
  MaterialExpressionNamedRerouteDeclaration_32 {
   class=MaterialExpressionNamedRerouteDeclaration
   inputs [None {from=MaterialExpressionMultiply_29}]
   properties {name="Soft Glow (V3)", nodeColor=<rgba 1 0 0.3984 1>}
  }
  MaterialExpressionNamedRerouteDeclaration_33 {
   class=MaterialExpressionNamedRerouteDeclaration
   inputs [None {from=MaterialExpressionScalarParameter_14}]
   properties {name="Breakup Pattern Mix (S)", nodeColor=<rgba 0 0.7188 1 1>}
  }
  MaterialExpressionNamedRerouteDeclaration_34 {
   class=MaterialExpressionNamedRerouteDeclaration
   inputs [None {from=MaterialExpressionScalarParameter_39}]
   properties {name="Compress Dynamic Range (S)", nodeColor=<rgba 1 0.2812 0 1>}
  }
  MaterialExpressionNamedRerouteDeclaration_35 {
   class=MaterialExpressionNamedRerouteDeclaration
   inputs [None {from=MaterialExpressionScalarParameter_40}]
   properties {name="Final Contrast (S)", nodeColor=<rgba 0 0.0391 1 1>}
  }
  MaterialExpressionNamedRerouteDeclaration_36 {
   class=MaterialExpressionNamedRerouteDeclaration
   inputs [None {from=MaterialExpressionScalarParameter_35}]
   properties {name="Vertical Attenuation Start V (S)", nodeColor=<rgba 0.4297 0 1 1>}
  }
  MaterialExpressionNamedRerouteDeclaration_37 {
   class=MaterialExpressionNamedRerouteDeclaration
   inputs [None {from=MaterialExpressionScalarParameter_34}]
   properties {name="Vertical Attenuation (S)", nodeColor=<rgba 0 1 0.1094 1>}
  }
  MaterialExpressionNamedRerouteDeclaration_38 {
   class=MaterialExpressionNamedRerouteDeclaration
   inputs [None {from=MaterialExpressionScalarParameter_37}]
   properties {name="Core Power (S)", nodeColor=<rgba 1 0 0.2109 1>}
  }
  MaterialExpressionNamedRerouteDeclaration_39 {
   class=MaterialExpressionNamedRerouteDeclaration
   inputs [None {from=MaterialExpressionScalarParameter_10}]
   properties {name="V Max (S)", nodeColor=<rgba 1 0.8906 0 1>}
  }
  MaterialExpressionNamedRerouteDeclaration_4 {
   class=MaterialExpressionNamedRerouteDeclaration
   inputs [None {from=MaterialExpressionVectorParameter_1, output=Per-Octave Freq Mult}]
   properties {name="Frequency Mult (S)", nodeColor=<rgba 0.1016 1 0 1>}
  }
  MaterialExpressionNamedRerouteDeclaration_40 {
   class=MaterialExpressionNamedRerouteDeclaration
   inputs [None {from=MaterialExpressionScalarParameter_9}]
   properties {name="V Min (S)", nodeColor=<rgba 0.5703 0 1 1>}
  }
  MaterialExpressionNamedRerouteDeclaration_41 {
   class=MaterialExpressionNamedRerouteDeclaration
   inputs [None {from=MaterialExpressionScalarParameter_16}]
   properties {name="U Stretch Mult (S)", nodeColor=<rgba 0 1 0.25 1>}
  }
  MaterialExpressionNamedRerouteDeclaration_42 {
   class=MaterialExpressionNamedRerouteDeclaration
   inputs [None {from=MaterialExpressionScalarParameter_17}]
   properties {name="V Stretch Mult (S)", nodeColor=<rgba 1 0 0.0703 1>}
  }
  MaterialExpressionNamedRerouteDeclaration_5 {
   class=MaterialExpressionNamedRerouteDeclaration
   inputs [None {from=MaterialExpressionScalarParameter_24}]
   properties {name="Base UV Spherize (S)", nodeColor=<rgba 0.5234 0 1 1>}
  }
  MaterialExpressionNamedRerouteDeclaration_6 {
   class=MaterialExpressionNamedRerouteDeclaration
   inputs [None {from=MaterialExpressionDivide_3}]
   properties {name="Warped UV (V2)", nodeColor=<rgba 1 0.75 0 1>}
  }
  MaterialExpressionNamedRerouteDeclaration_7 {
   class=MaterialExpressionNamedRerouteDeclaration
   inputs [None {from=MaterialExpressionMultiply_2}]
   properties {name="Time (S)", nodeColor=<rgba 0.4297 0 1 1>}
  }
  MaterialExpressionNamedRerouteDeclaration_8 {
   class=MaterialExpressionNamedRerouteDeclaration
   inputs [None {from=MaterialExpressionScalarParameter_25}]
   properties {name="Base UV Scale (S)", nodeColor=<rgba 0 1 0.2031 1>}
  }
  MaterialExpressionNamedRerouteDeclaration_9 {
   class=MaterialExpressionNamedRerouteDeclaration
   inputs [None {from=MaterialExpressionScalarParameter_21}]
   properties {name="Base UV Rotate (S)", nodeColor=<rgba 1 0 0.1172 1>}
  }
  MaterialExpressionNamedRerouteUsage_0 {class=MaterialExpressionNamedRerouteUsage}
  MaterialExpressionNamedRerouteUsage_1 {class=MaterialExpressionNamedRerouteUsage}
  MaterialExpressionNamedRerouteUsage_10 {class=MaterialExpressionNamedRerouteUsage}
  MaterialExpressionNamedRerouteUsage_11 {class=MaterialExpressionNamedRerouteUsage}
  MaterialExpressionNamedRerouteUsage_12 {class=MaterialExpressionNamedRerouteUsage}
  MaterialExpressionNamedRerouteUsage_13 {class=MaterialExpressionNamedRerouteUsage}
  MaterialExpressionNamedRerouteUsage_14 {class=MaterialExpressionNamedRerouteUsage}
  MaterialExpressionNamedRerouteUsage_15 {class=MaterialExpressionNamedRerouteUsage}
  MaterialExpressionNamedRerouteUsage_16 {class=MaterialExpressionNamedRerouteUsage}
  MaterialExpressionNamedRerouteUsage_17 {class=MaterialExpressionNamedRerouteUsage}
  MaterialExpressionNamedRerouteUsage_18 {class=MaterialExpressionNamedRerouteUsage}
  MaterialExpressionNamedRerouteUsage_19 {class=MaterialExpressionNamedRerouteUsage}
  MaterialExpressionNamedRerouteUsage_2 {class=MaterialExpressionNamedRerouteUsage}
  MaterialExpressionNamedRerouteUsage_20 {class=MaterialExpressionNamedRerouteUsage}
  MaterialExpressionNamedRerouteUsage_21 {class=MaterialExpressionNamedRerouteUsage}
  MaterialExpressionNamedRerouteUsage_22 {class=MaterialExpressionNamedRerouteUsage}
  MaterialExpressionNamedRerouteUsage_23 {class=MaterialExpressionNamedRerouteUsage}
  MaterialExpressionNamedRerouteUsage_24 {class=MaterialExpressionNamedRerouteUsage}
  MaterialExpressionNamedRerouteUsage_25 {class=MaterialExpressionNamedRerouteUsage}
  MaterialExpressionNamedRerouteUsage_26 {class=MaterialExpressionNamedRerouteUsage}
  MaterialExpressionNamedRerouteUsage_27 {class=MaterialExpressionNamedRerouteUsage}
  MaterialExpressionNamedRerouteUsage_28 {class=MaterialExpressionNamedRerouteUsage}
  MaterialExpressionNamedRerouteUsage_29 {class=MaterialExpressionNamedRerouteUsage}
  MaterialExpressionNamedRerouteUsage_3 {class=MaterialExpressionNamedRerouteUsage}
  MaterialExpressionNamedRerouteUsage_30 {class=MaterialExpressionNamedRerouteUsage}
  MaterialExpressionNamedRerouteUsage_31 {class=MaterialExpressionNamedRerouteUsage}
  MaterialExpressionNamedRerouteUsage_32 {class=MaterialExpressionNamedRerouteUsage}
  MaterialExpressionNamedRerouteUsage_33 {class=MaterialExpressionNamedRerouteUsage}
  MaterialExpressionNamedRerouteUsage_34 {class=MaterialExpressionNamedRerouteUsage}
  MaterialExpressionNamedRerouteUsage_35 {class=MaterialExpressionNamedRerouteUsage}
  MaterialExpressionNamedRerouteUsage_36 {class=MaterialExpressionNamedRerouteUsage}
  MaterialExpressionNamedRerouteUsage_37 {class=MaterialExpressionNamedRerouteUsage}
  MaterialExpressionNamedRerouteUsage_38 {class=MaterialExpressionNamedRerouteUsage}
  MaterialExpressionNamedRerouteUsage_39 {class=MaterialExpressionNamedRerouteUsage}
  MaterialExpressionNamedRerouteUsage_4 {class=MaterialExpressionNamedRerouteUsage}
  MaterialExpressionNamedRerouteUsage_40 {class=MaterialExpressionNamedRerouteUsage}
  MaterialExpressionNamedRerouteUsage_41 {class=MaterialExpressionNamedRerouteUsage}
  MaterialExpressionNamedRerouteUsage_42 {class=MaterialExpressionNamedRerouteUsage}
  MaterialExpressionNamedRerouteUsage_44 {class=MaterialExpressionNamedRerouteUsage}
  MaterialExpressionNamedRerouteUsage_45 {class=MaterialExpressionNamedRerouteUsage}
  MaterialExpressionNamedRerouteUsage_46 {class=MaterialExpressionNamedRerouteUsage}
  MaterialExpressionNamedRerouteUsage_47 {class=MaterialExpressionNamedRerouteUsage}
  MaterialExpressionNamedRerouteUsage_48 {class=MaterialExpressionNamedRerouteUsage}
  MaterialExpressionNamedRerouteUsage_49 {class=MaterialExpressionNamedRerouteUsage}
  MaterialExpressionNamedRerouteUsage_5 {class=MaterialExpressionNamedRerouteUsage}
  MaterialExpressionNamedRerouteUsage_50 {class=MaterialExpressionNamedRerouteUsage}
  MaterialExpressionNamedRerouteUsage_51 {class=MaterialExpressionNamedRerouteUsage}
  MaterialExpressionNamedRerouteUsage_52 {class=MaterialExpressionNamedRerouteUsage}
  MaterialExpressionNamedRerouteUsage_53 {class=MaterialExpressionNamedRerouteUsage}
  MaterialExpressionNamedRerouteUsage_54 {class=MaterialExpressionNamedRerouteUsage}
  MaterialExpressionNamedRerouteUsage_55 {class=MaterialExpressionNamedRerouteUsage}
  MaterialExpressionNamedRerouteUsage_57 {class=MaterialExpressionNamedRerouteUsage}
  MaterialExpressionNamedRerouteUsage_59 {class=MaterialExpressionNamedRerouteUsage}
  MaterialExpressionNamedRerouteUsage_6 {class=MaterialExpressionNamedRerouteUsage}
  MaterialExpressionNamedRerouteUsage_60 {class=MaterialExpressionNamedRerouteUsage}
  MaterialExpressionNamedRerouteUsage_61 {class=MaterialExpressionNamedRerouteUsage}
  MaterialExpressionNamedRerouteUsage_62 {class=MaterialExpressionNamedRerouteUsage}
  MaterialExpressionNamedRerouteUsage_63 {class=MaterialExpressionNamedRerouteUsage}
  MaterialExpressionNamedRerouteUsage_64 {class=MaterialExpressionNamedRerouteUsage}
  MaterialExpressionNamedRerouteUsage_9 {class=MaterialExpressionNamedRerouteUsage}
  MaterialExpressionOneMinus_3 {class=MaterialExpressionOneMinus, inputs [None {from=MaterialExpressionConvert_4}], properties {desc=Flip V}}
  MaterialExpressionOneMinus_4 {class=MaterialExpressionOneMinus, inputs [None {from=MaterialExpressionDivide_2}], properties {desc=v stretch}}
  MaterialExpressionOneMinus_9 {class=MaterialExpressionOneMinus, inputs [None {from=MaterialExpressionExponential_1}]}
  MaterialExpressionPower_4 {
   class=MaterialExpressionPower
   inputs [
    Base {from=MaterialExpressionDivide_9}
    Exp {from=MaterialExpressionNamedRerouteUsage_5}
   ]
   properties {constExponent=5}
  }
  MaterialExpressionPower_5 {
   class=MaterialExpressionPower
   inputs [
    Base {from=MaterialExpressionAdd_5}
    Exp {from=MaterialExpressionNamedRerouteUsage_26}
   ]
   properties {constExponent=3}
  }
  MaterialExpressionPower_6 {
   class=MaterialExpressionPower
   inputs [
    Base {from=MaterialExpressionLinearInterpolate_8}
    Exp {from=MaterialExpressionNamedRerouteUsage_25}
   ]
  }
  MaterialExpressionScalarParameter_0 {
   class=MaterialExpressionScalarParameter
   properties {defaultValue=1, group=Fire | 04 - Luminance, parameterName=Fire - Core Brightness, sortPriority=30}
  }
  MaterialExpressionScalarParameter_10 {
   class=MaterialExpressionScalarParameter
   properties {defaultValue=2, group=Fire | 01 - UV, parameterName=Fire - Smoothstep V Max, sortPriority=120}
  }
  MaterialExpressionScalarParameter_14 {
   class=MaterialExpressionScalarParameter
   properties {defaultValue=1, group=Fire | 03 - Shaping, parameterName=Fire - Breakup Pattern Mix, sliderMax=1, sortPriority=60}
  }
  MaterialExpressionScalarParameter_15 {
   class=MaterialExpressionScalarParameter
   properties {defaultValue=1, group=Fire | 01 - UV, parameterName=Fire - Warped UV Scale, sortPriority=200}
  }
  MaterialExpressionScalarParameter_16 {
   class=MaterialExpressionScalarParameter
   properties {defaultValue=1, group=Fire | 01 - UV, parameterName=Fire - U Stretch Mult, sortPriority=130}
  }
  MaterialExpressionScalarParameter_17 {
   class=MaterialExpressionScalarParameter
   properties {defaultValue=1, group=Fire | 01 - UV, parameterName=Fire - V Stretch Mult, sortPriority=140}
  }
  MaterialExpressionScalarParameter_21 {
   class=MaterialExpressionScalarParameter
   properties {group=Fire | 01 - UV, parameterName=Fire - UV Rotate, sliderMax=1, sliderMin=-1, sortPriority=30}
  }
  MaterialExpressionScalarParameter_24 {
   class=MaterialExpressionScalarParameter
   properties {group=Fire | 01 - UV, parameterName=Fire - UV Spherize, sliderMax=10, sliderMin=-10, sortPriority=10}
  }
  MaterialExpressionScalarParameter_25 {class=MaterialExpressionScalarParameter, properties {defaultValue=1, group=Fire | 01 - UV, parameterName=Fire - UV Scale, sortPriority=20}}
  MaterialExpressionScalarParameter_26 {
   class=MaterialExpressionScalarParameter
   properties {group=Fire | 03 - Shaping, parameterName=Fire - Core Radius, sliderMax=10, sortPriority=10}
  }
  MaterialExpressionScalarParameter_3 {
   class=MaterialExpressionScalarParameter
   properties {defaultValue=6, group=Fire | 02 - Time, parameterName=Fire - Turbulence Speed, sortPriority=20}
  }
  MaterialExpressionScalarParameter_34 {
   class=MaterialExpressionScalarParameter
   properties {group=Fire | 03 - Shaping, parameterName=Fire - Vertical Attenuation, sliderMax=1, sortPriority=20}
  }
  MaterialExpressionScalarParameter_35 {
   class=MaterialExpressionScalarParameter
   properties {defaultValue=-0.5, group=Fire | 03 - Shaping, parameterName=Fire - Attenuation Start V, sortPriority=30}
  }
  MaterialExpressionScalarParameter_37 {
   class=MaterialExpressionScalarParameter
   properties {defaultValue=3, group=Fire | 04 - Luminance, parameterName=Fire - Core Power, sortPriority=20}
  }
  MaterialExpressionScalarParameter_38 {
   class=MaterialExpressionScalarParameter
   properties {defaultValue=1, group=Fire | 04 - Luminance, parameterName=Fire - Soft Glow Intensity, sortPriority=40}
  }
  MaterialExpressionScalarParameter_39 {
   class=MaterialExpressionScalarParameter
   properties {group=Fire | 04 - Luminance, parameterName=Fire - Compress Dynamic Range, sliderMax=1, sortPriority=50}
  }
  MaterialExpressionScalarParameter_40 {
   class=MaterialExpressionScalarParameter
   properties {defaultValue=1, group=Fire | 04 - Luminance, parameterName=Fire - Final Contrast, sortPriority=60}
  }
  MaterialExpressionScalarParameter_42 {
   class=MaterialExpressionScalarParameter
   properties {defaultValue=10, group=Fire | 03 - Shaping, parameterName=Fire - Breakup Pattern Scale, sortPriority=50}
  }
  MaterialExpressionScalarParameter_7 {
   class=MaterialExpressionScalarParameter
   properties {defaultValue=0.25, group=Fire | 02 - Time, parameterName=Fire - Scroll Speed, sortPriority=30}
  }
  MaterialExpressionScalarParameter_8 {
   class=MaterialExpressionScalarParameter
   properties {defaultValue=1, group=Fire | 02 - Time, parameterName=Fire - Global Time Mult, sortPriority=10}
  }
  MaterialExpressionScalarParameter_9 {
   class=MaterialExpressionScalarParameter
   properties {defaultValue=-2, group=Fire | 01 - UV, parameterName=Fire - Smoothstep V Min, sortPriority=110}
  }
  MaterialExpressionSmoothStep_0 {
   class=MaterialExpressionSmoothStep
   inputs [
    Min {from=MaterialExpressionNamedRerouteUsage_37, output=""}
    Max {from=MaterialExpressionNamedRerouteUsage_51, output=""}
    Value {from=MaterialExpressionComponentMask_3, output=""}
   ]
   properties {}
  }
  MaterialExpressionSubtract_10 {
   class=MaterialExpressionSubtract
   inputs [
    A {from=MaterialExpressionComponentMask_10}
    B {from=MaterialExpressionNamedRerouteUsage_30}
   ]
  }
  MaterialExpressionSubtract_3 {
   class=MaterialExpressionSubtract
   inputs [
    A {from=""}
    B {from=MaterialExpressionMultiply_6}
   ]
   properties {constA=2, desc=u stretch}
  }
  MaterialExpressionSubtract_4 {
   class=MaterialExpressionSubtract
   inputs [
    A {from=MaterialExpressionNamedRerouteUsage_6}
    B {from=MaterialExpressionNamedRerouteUsage_14}
   ]
   properties {desc=Add scroll to turbulence}
  }
  MaterialExpressionSubtract_9 {
   class=MaterialExpressionSubtract
   inputs [
    A {from=MaterialExpressionMultiply_13}
    B {from=MaterialExpressionNamedRerouteUsage_34}
   ]
  }
  MaterialExpressionTextureCoordinate_0 {class=MaterialExpressionTextureCoordinate}
  MaterialExpressionTextureObjectParameter_0 {
   class=MaterialExpressionTextureObjectParameter
   inputs [
    UVs {from="", output=""}
    Apply View MipBias {from="", output=""}
   ]
   properties {
    group=Fire | 03 - Shaping
    parameterName=Fire - Breakup Pattern Map
    samplerType=SAMPLERTYPE_LinearColor
    sortPriority=40
    texture=/BRCosmetics/Animation/Props/Gadgets_OutlineLove/Textures/T_Fire_Breakup
   }
  }
  MaterialExpressionTextureSample_1 {
   class=MaterialExpressionTextureSample
   inputs [
    UVs {from=MaterialExpressionDivide_16}
    Tex {from=MaterialExpressionNamedRerouteUsage_21}
    Apply View MipBias {from=""}
   ]
   properties {samplerSource=SSM_Wrap_WorldGroupSettings, texture=&T_BlacksRGB}
  }
  MaterialExpressionTime_0 {class=MaterialExpressionTime}
  MaterialExpressionVectorParameter_0 {
   class=MaterialExpressionVectorParameter
   properties {
    group=Fire | 04 - Luminance
    parameterName=Fire - Color
    sortPriority=10
    channelNames {a=n/a, b=Fire Color B, g=Fire Color G, r=Fire Color R}
    defaultValue=<rgba 1 0.225 0.1125 0>
   }
  }
  MaterialExpressionVectorParameter_1 {
   class=MaterialExpressionVectorParameter
   properties {
    group=Fire | 01 - UV
    parameterName=Fire - UV Turbulence
    sortPriority=300
    channelNames {a=Per-Octave Freq Mult, b=Base Frequency, g=Wave Amplitude, r=Num Octaves}
    defaultValue=<rgba 10 0.4 7 1.3>
   }
  }
 }}
