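# Masked, two-sided material for the OutlineLove gadget prop, built between the shared 2023 character parent PreFX and PostFX functions.
# Tech art mask channels supply a sigil mask and a dithered opacity mask shaped by a power parameter.
# A static switch enables blending in a black-base, fire-emissive sigil layer.
# That emission multiplies the emissive map by the OutlineLove fire function, with remapped sigil UVs defined alongside.
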Material /BRCosmetics/Animation/Props/Gadgets_OutlineLove/Materials/M_Gadget_OutlineLove
graph {
 expressions {
  MaterialExpressionAdd_9 {
   class=MaterialExpressionAdd
   inputs [
    A {from=MaterialExpressionMultiply_0}
    B {from=""}
   ]
  }
  MaterialExpressionBlendMaterialAttributes_0 {
   class=MaterialExpressionBlendMaterialAttributes
   inputs [
    A {from=MaterialExpressionReroute_0, output=""}
    B {from=MaterialExpressionSetMaterialAttributes_2, output=""}
    Alpha {from=MaterialExpressionNamedRerouteUsage_36, output=""}
   ]
   properties {}
  }
  MaterialExpressionComponentMask_2 {class=MaterialExpressionComponentMask, inputs [None {from=MaterialExpressionNamedRerouteDeclaration_22}], properties {g=true}}
  MaterialExpressionComponentMask_5 {class=MaterialExpressionComponentMask, inputs [None {from=MaterialExpressionNamedRerouteDeclaration_22}], properties {r=true}}
  MaterialExpressionConstant_0 {class=MaterialExpressionConstant}
  MaterialExpressionConstant_10 {class=MaterialExpressionConstant, properties {r=0.23}}
  MaterialExpressionConstant_11 {class=MaterialExpressionConstant, properties {r=0.78}}
  MaterialExpressionConstant_6 {class=MaterialExpressionConstant}
  MaterialExpressionConstant_9 {class=MaterialExpressionConstant, properties {r=1}}
  MaterialExpressionConvert_1 {
   class=MaterialExpressionConvert
   inputs [
    None {from=MaterialExpressionMaterialFunctionCall_3, output=Result}
    None {from=MaterialExpressionMaterialFunctionCall_5, output=Result}
   ]
   properties {
    nodeName=Make Vector2
    convertInputs [
     {
      type=Scalar
      expressionInput {
       expression=&self:MaterialExpressionMaterialFunctionCall_3
       inputName=None
       mask=0
       maskA=0
       maskB=0
       maskG=0
       maskR=0
       outputIndex=0
      }
     }
     {
      type=Scalar
      expressionInput {
       expression=&self:MaterialExpressionMaterialFunctionCall_5
       inputName=None
       mask=0
       maskA=0
       maskB=0
       maskG=0
       maskR=0
       outputIndex=0
      }
     }
    ]
    convertOutputs [{type=Vector2}]
   }
  }
  MaterialExpressionConvert_7 {
   class=MaterialExpressionConvert
   inputs [None {from=MaterialExpressionTextureCoordinate_0}]
   properties {
    nodeName=Break Float2
    convertInputs [{
      type=Vector2
      expressionInput {
       expression=&self:MaterialExpressionTextureCoordinate_0
       inputName=None
       mask=0
       maskA=0
       maskB=0
       maskG=0
       maskR=0
       outputIndex=0
      }
     }]
    convertOutputs [{type=Scalar}, {type=Scalar}]
   }
  }
  MaterialExpressionGetMaterialAttributes_1 {
   class=MaterialExpressionGetMaterialAttributes
   inputs [None {from=MaterialExpressionSetMaterialAttributes_0, output=""}]
   properties {attributeGetTypes [#1]}
  }
  MaterialExpressionLinearInterpolate_1 {
   class=MaterialExpressionLinearInterpolate
   inputs [
    A {from=MaterialExpressionGetMaterialAttributes_1, output=Opacity Mask}
    B {from=""}
    Alpha {from=MaterialExpressionScalarParameter_13}
   ]
  }
  MaterialExpressionMaterialFunctionCall_0 {
   class=MaterialExpressionMaterialFunctionCall
   inputs [
    Overide Texture Set UVs {from=""}
    Override Tech Art Mask UVs {from=""}
    Override Basic Emissive UVs {from=""}
   ]
   properties {materialFunction=&MF_CharacterParent_2023_PreFX}
  }
  MaterialExpressionMaterialFunctionCall_1 {
   class=MaterialExpressionMaterialFunctionCall
   inputs [
    Material {from=MaterialExpressionStaticSwitchParameter_1}
    M {from=MaterialExpressionMaterialFunctionCall_0, output=M}
    Pre Skinned Local Position {from=MaterialExpressionMaterialFunctionCall_0, output=M}
    Distance Blend {from=MaterialExpressionMaterialFunctionCall_0, output=M}
    Fresnel {from=MaterialExpressionMaterialFunctionCall_0, output=M}
   ]
   properties {materialFunction=&MF_CharacterParent_2023_PostFX}
  }
  MaterialExpressionMaterialFunctionCall_11 {
   class=MaterialExpressionMaterialFunctionCall
   inputs [UV {from=MaterialExpressionNamedRerouteUsage_59}]
   properties {materialFunction=&/MF_Gadget_OutlineLove_Fire}
  }
  MaterialExpressionMaterialFunctionCall_3 {
   class=MaterialExpressionMaterialFunctionCall
   inputs [
    Input {from=MaterialExpressionReroute_2}
    Input Low {from=MaterialExpressionConstant_6}
    Input High {from=MaterialExpressionConstant_11}
   ]
   properties {materialFunction=/Engine/Functions/Engine_MaterialFunctions03/Math/RemapValueRangeNormalized}
  }
  MaterialExpressionMaterialFunctionCall_5 {
   class=MaterialExpressionMaterialFunctionCall
   inputs [
    Input {from=MaterialExpressionReroute_3}
    Input Low {from=MaterialExpressionConstant_10}
    Input High {from=MaterialExpressionConstant_9}
   ]
   properties {materialFunction=&RemapValueRangeNormalized}
  }
  MaterialExpressionMultiply_0 {
   class=MaterialExpressionMultiply
   inputs [
    A {from=MaterialExpressionTextureSampleParameter2D_1, output=RGB}
    B {from=MaterialExpressionScalarParameter_11}
   ]
  }
  MaterialExpressionMultiply_14 {
   class=MaterialExpressionMultiply
   inputs [
    A {from=MaterialExpressionAdd_9}
    B {from=MaterialExpressionMaterialFunctionCall_11, output="Result (V3)"}
   ]
  }
  MaterialExpressionNamedRerouteDeclaration_13 {
   class=MaterialExpressionNamedRerouteDeclaration
   inputs [None {from=MaterialExpressionConvert_1}]
   properties {name="Sigil UV (V2)", nodeColor=<rgba 1 0 0.6094 1>}
  }
  MaterialExpressionNamedRerouteDeclaration_21 {
   class=MaterialExpressionNamedRerouteDeclaration
   inputs [None {from=MaterialExpressionMultiply_14}]
   properties {name="Sigil Emission (V3)", nodeColor=<rgba 1 0.3148 0.0025 1>}
  }
  MaterialExpressionNamedRerouteDeclaration_22 {
   class=MaterialExpressionNamedRerouteDeclaration
   inputs [None {from=MaterialExpressionMaterialFunctionCall_0, output=M}]
   properties {name="TechArt Mask (V3)", nodeColor=<rgba 0.1016 0 1 1>}
  }
  MaterialExpressionNamedRerouteDeclaration_23 {
   class=MaterialExpressionNamedRerouteDeclaration
   inputs [None {from=MaterialExpressionComponentMask_5}]
   properties {name="Sigil Mask (S)", nodeColor=<rgba 0 0.5913 1 1>}
  }
  MaterialExpressionNamedRerouteDeclaration_24 {
   class=MaterialExpressionNamedRerouteDeclaration
   inputs [None {from=MaterialExpressionComponentMask_2}]
   properties {name="Opacity Mask (S)", nodeColor=<rgba 0 0.5913 1 1>}
  }
  MaterialExpressionNamedRerouteDeclaration_34 {
   class=MaterialExpressionNamedRerouteDeclaration
   inputs [None {from=MaterialExpressionMaterialFunctionCall_11, output="Turbulence UV (Static) (V2)"}]
   properties {name="Turbulence UV (Static) (V2)", nodeColor=<rgba 0.6641 1 0 1>}
  }
  MaterialExpressionNamedRerouteUsage_26 {class=MaterialExpressionNamedRerouteUsage}
  MaterialExpressionNamedRerouteUsage_36 {class=MaterialExpressionNamedRerouteUsage}
  MaterialExpressionNamedRerouteUsage_37 {class=MaterialExpressionNamedRerouteUsage}
  MaterialExpressionNamedRerouteUsage_59 {class=MaterialExpressionNamedRerouteUsage}
  MaterialExpressionPower_0 {
   class=MaterialExpressionPower
   inputs [
    Base {from=MaterialExpressionNamedRerouteUsage_37}
    Exp {from=MaterialExpressionScalarParameter_0}
   ]
  }
  MaterialExpressionReroute_0 {class=MaterialExpressionReroute, inputs [None {from=MaterialExpressionReroute_6}]}
  MaterialExpressionReroute_1 {class=MaterialExpressionReroute, inputs [None {from=MaterialExpressionMaterialFunctionCall_0, output=Result}]}
  MaterialExpressionReroute_2 {class=MaterialExpressionReroute, inputs [None {from=MaterialExpressionConvert_7}]}
  MaterialExpressionReroute_3 {class=MaterialExpressionReroute, inputs [None {from=MaterialExpressionConvert_7}]}
  MaterialExpressionReroute_6 {class=MaterialExpressionReroute, inputs [None {from=MaterialExpressionSetMaterialAttributes_0}]}
  MaterialExpressionScalarParameter_0 {
   class=MaterialExpressionScalarParameter
   properties {defaultValue=0.7, group=03_Opacity Mask, parameterName=Dragon Opacity Pow, sliderMax=1, sortPriority=1010}
  }
  MaterialExpressionScalarParameter_11 {
   class=MaterialExpressionScalarParameter
   properties {defaultValue=1, group=Fire | 00 - Sigil Emission, parameterName=Sigil - Emissive Map Gain, sliderMax=10, sortPriority=10}
  }
  MaterialExpressionScalarParameter_13 {
   class=MaterialExpressionScalarParameter
   properties {group=Fire | 00 - Sigil Emission, parameterName=TMP - Force Opaque, sliderMax=1, sortPriority=100}
  }
  MaterialExpressionSetMaterialAttributes_0 {
   class=MaterialExpressionSetMaterialAttributes
   inputs [
    MaterialAttributes {from=MaterialExpressionReroute_1, output=""}
    Opacity Mask {from=MaterialExpressionPower_0, output=""}
   ]
   properties {attributeSetTypes [#1]}
  }
  MaterialExpressionSetMaterialAttributes_2 {
   class=MaterialExpressionSetMaterialAttributes
   inputs [
    MaterialAttributes {from=MaterialExpressionReroute_6, output=""}
    Emissive Color {from=MaterialExpressionNamedRerouteUsage_26, output=""}
    Base Color {from=MaterialExpressionConstant_0, output=""}
    Specular {from=MaterialExpressionConstant_0, output=""}
   ]
   properties {attributeSetTypes [#2, #3, #4]}
  }
  MaterialExpressionStaticSwitchParameter_1 {
   class=MaterialExpressionStaticSwitchParameter
   inputs [
    True {from=MaterialExpressionBlendMaterialAttributes_0, output=""}
    False {from=MaterialExpressionReroute_0, output=""}
   ]
   properties {group=Fire | 00 - Sigil Emission, parameterName=Sigil - Enable Fire, sortPriority=0}
  }
  MaterialExpressionTextureCoordinate_0 {class=MaterialExpressionTextureCoordinate}
  MaterialExpressionTextureSampleParameter2D_1 {
   class=MaterialExpressionTextureSampleParameter2D
   inputs [
    UVs {from="", output=""}
    Apply View MipBias {from="", output=""}
   ]
   properties {
    group=06_Emissive
    parameterName=Emissive
    samplerSource=SSM_Wrap_WorldGroupSettings
    sortPriority=1
    texture=&T_BlacksRGB
   }
  }
 }
 properties {
  bHasPixelAnimation=true
  bUseMaterialAttributes=true
  bUsedWithClothing=true
  bUsedWithMeshDeformer=true
  bUsedWithMorphTargets=true
  bUsedWithNanite=true
  bUsedWithNiagaraMeshParticles=true
  bUsedWithSkeletalMesh=true
  blendMode=BLEND_Masked
  ditherOpacityMask=true
  opacityMaskClipValue=0.5
  pixelDepthOffsetMode=PDOM_Legacy
  shadingModel=MSM_FromMaterialExpression
  twoSided=true
  usedShadingModels=Default Lit | Subsurface | Clear Coat | Subsurface Profile | Cloth
 }
}
